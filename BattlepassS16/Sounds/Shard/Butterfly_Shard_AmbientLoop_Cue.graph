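# Sound cue for the ambient loop of a Battle Pass S16 butterfly shard, in the quest sound class with short-range overridden supply drop marker attenuation.
# The first two quality slots mix a looping supply drop target sound, a one-shot random crystal gust and a modulated bed of three looping, randomized crystal gust layers.
# Each bed layer has its own pitch range, two via delay nodes.
# The last two slots play only an optimized loop.
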SoundCue /BattlepassS16/Sounds/Shard/Butterfly_Shard_AmbientLoop_Cue
root=QualityLevel_0
nodes [
 {id=Delay_0, children [Random_3], properties {delayMax=0.5, delayMin=0.2}}
 {id=Delay_2, children [Random_2]}
 {id=Looping_1, children [Delay_2]}
 {id=Looping_2, children [Modulator_2]}
 {id=Looping_3, children [Modulator_1]}
 {id=Mixer_1, children [WavePlayer_0,Random_3,Modulator_0], properties {inputVolume [1,1.3,1.3]}}
 {id=Mixer_2, children [Looping_3,Looping_2,Modulator_3], properties {inputVolume [1,0.75,0.5]}}
 {id=Modulator_0, children [Mixer_2], properties {pitchMax=1.1, pitchMin=0.7, volumeMax=1.1, volumeMin=0.9}}
 {id=Modulator_1, children [Delay_0], properties {pitchMax=1, pitchMin=0.5}}
 {id=Modulator_2, children [Random_0], properties {pitchMax=2, pitchMin=1.7, volumeMin=0.5}}
 {id=Modulator_3, children [Looping_1], properties {pitchMax=1.3, pitchMin=1.2, volumeMin=0.7}}
 {id=QualityLevel_0, children [Mixer_1,Mixer_1,WavePlayer_1,WavePlayer_1]}
 {id=Random_0, children [WavePlayer_30, WavePlayer_31, WavePlayer_32], properties {weights [1,1,1]}}
 {id=Random_2, children [WavePlayer_2,WavePlayer_3,WavePlayer_4], properties {weights [1,1,1]}}
 {id=Random_3, children [WavePlayer_35, WavePlayer_36, WavePlayer_37], properties {weights [1,1,1]}}
 {id=WavePlayer_0, properties {bLooping=true, soundWaveAssetPtr=/Game/Sounds/Foley_Loot/Containers/SupplyDrops/sfx_supplydrop_target_loop}}
 {id=WavePlayer_1, properties {bLooping=true, soundWaveAssetPtr=/Game/Sounds/Foley_Loot/Containers/SupplyDrops/sfx_supplydrop_target_loop_opt}}
 {id=WavePlayer_2, properties {soundWaveAssetPtr=/Game/Sounds/Apollo/Biomes/Desert/AMB_SS_S15_Desert_WindGust_Magic_Crystal_04}}
 {id=WavePlayer_3, properties {soundWaveAssetPtr=/Game/Sounds/Apollo/Biomes/Desert/AMB_SS_S15_Desert_WindGust_Magic_Crystal_05}}
 {id=WavePlayer_30, properties {soundWaveAssetPtr=&AMB_SS_S15_Desert_WindGust_Magic_Crystal_04}}
 {id=WavePlayer_31, properties {soundWaveAssetPtr=&AMB_SS_S15_Desert_WindGust_Magic_Crystal_05}}
 {id=WavePlayer_32, properties {soundWaveAssetPtr=/Game/Sounds/Apollo/Biomes/Desert/AMB_SS_S15_Desert_WindGust_Magic_Crystal_06}}
 {id=WavePlayer_35, properties {soundWaveAssetPtr=/Game/Sounds/Apollo/Biomes/Desert/AMB_SS_S15_Desert_WindGust_Magic_Crystal_01}}
 {id=WavePlayer_36, properties {soundWaveAssetPtr=/Game/Sounds/Apollo/Biomes/Desert/AMB_SS_S15_Desert_WindGust_Magic_Crystal_02}}
 {id=WavePlayer_37, properties {soundWaveAssetPtr=/Game/Sounds/Apollo/Biomes/Desert/AMB_SS_S15_Desert_WindGust_Magic_Crystal_03}}
 {id=WavePlayer_4, properties {soundWaveAssetPtr=&AMB_SS_S15_Desert_WindGust_Magic_Crystal_06}}
]
properties {
 attenuationSettings=/Game/Sounds/Foley_Loot/Containers/SupplyDrops/Fort_SupplyDrop_Marker_Close_Athena_Att
 bOverrideAttenuation=true
 maxDistance=800
 pitchMultiplier=2
 soundClassObject=&Fort_Quests
 volumeMultiplier=1.568
 attenuationOverrides {falloffDistance=700, attenuationShapeExtents {x=100}}
}
