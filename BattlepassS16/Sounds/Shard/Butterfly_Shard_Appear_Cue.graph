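# Sound Cue named for the Butterfly Shard appear sound in the Season 16 Battle Pass content, classed under quest audio.
# The file shows no nodes and a null root, so it references no waves to play.
# It uses the shared S16 narrative familiar attenuation asset, boosts volume, and overrides concurrency to allow few simultaneous instances.
# Inline attenuation values are stored, but their override flag is not enabled.
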SoundCue /BattlepassS16/Sounds/Shard/Butterfly_Shard_Appear_Cue
root=null
nodes []
properties {
 attenuationSettings=/Game/Athena/Sounds/Quests/S16_Narrative/S16_Narrative_Familiar_Att
 bOverrideConcurrency=true
 maxDistance=3012
 soundClassObject=&Fort_Quests
 volumeMultiplier=1.792
 attenuationOverrides {falloffDistance=3000, attenuationShapeExtents {x=700}}
 concurrencyOverrides {maxCount=3}
}
