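# Sound cue for collecting a butterfly shard in BattlepassS16, routed to the non-critical in-game UI sound class.
# A mixer layers a quieted reward-screen chest upgrade sound, a boosted magic crystal wind gust and a butterfly spawn-in sound.
# The gust is picked at random from three equally weighted variants and pitch-modulated.
# A concurrency override allows one instance at a time, stopping the oldest.
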SoundCue /BattlepassS16/Sounds/Shard/Butterfly_Shard_Collect_Cue
root=Mixer_0
nodes [
 {id=Mixer_0, children [WavePlayer_1,Modulator_0,WavePlayer_2], properties {inputVolume [0.5,2.5,1]}}
 {id=Modulator_0, children [Random_3], properties {pitchMax=2, volumeMax=1, volumeMin=1}}
 {id=Random_3, children [WavePlayer_35, WavePlayer_36, WavePlayer_37], properties {weights [1,1,1]}}
 {id=WavePlayer_1, properties {soundWaveAssetPtr=/Game/Sounds/Fort_UI_Sounds/Rewards/Fort_RewardScreen_TeamScore_ChestLootUpgrade}}
 {id=WavePlayer_2, properties {soundWaveAssetPtr=/Game/Athena/Sounds/Quests/S16_Narrative/Butterfly_Spawn_In_01}}
 {id=WavePlayer_35, properties {soundWaveAssetPtr=/Game/Sounds/Apollo/Biomes/Desert/AMB_SS_S15_Desert_WindGust_Magic_Crystal_01}}
 {id=WavePlayer_36, properties {soundWaveAssetPtr=/Game/Sounds/Apollo/Biomes/Desert/AMB_SS_S15_Desert_WindGust_Magic_Crystal_02}}
 {id=WavePlayer_37, properties {soundWaveAssetPtr=/Game/Sounds/Apollo/Biomes/Desert/AMB_SS_S15_Desert_WindGust_Magic_Crystal_03}}
]
properties {
 bOverrideConcurrency=true
 maxDistance=2097152
 soundClassObject=&Fort_UI_InGame_NonCritical
 volumeMultiplier=1
 concurrencyOverrides {maxCount=1, resolutionRule=StopOldest}
}
